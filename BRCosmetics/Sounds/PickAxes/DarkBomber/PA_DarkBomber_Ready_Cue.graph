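# Sound cue for the DarkBomber pickaxe's ready sound in Battle Royale cosmetics.
# An equally weighted random node picks one of three DarkUnicorn swing waves and feeds a modulator at the root.
# The cue reuses those swing recordings at raised pitch and low volume, classed as melee weapon audio with close-range projectile attenuation.
# It primes on load, references a pickaxe-ready property sheet, and overrides concurrency to prevent new instances at the limit.
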SoundCue /BRCosmetics/Sounds/PickAxes/DarkBomber/PA_DarkBomber_Ready_Cue
root=Modulator_1
nodes [
 {id=Modulator_1, children [Random_1]}
 {id=Random_1, children [WavePlayer_3,WavePlayer_4,WavePlayer_5], properties {weights [1,1,1]}}
 {id=WavePlayer_3, properties {soundWaveAssetPtr=&/PA_DarkUnicorn_Swing_01}}
 {id=WavePlayer_4, properties {soundWaveAssetPtr=&/PA_DarkUnicorn_Swing_02}}
 {id=WavePlayer_5, properties {soundWaveAssetPtr=&/PA_DarkUnicorn_Swing_03}}
]
properties {
 attenuationSettings=&Fort_Weapon_Projectile_Close_Athena_Att
 audioPropertiesSheet=/BRCosmetics/Sounds/Settings/PropertySheets/Pickaxe/APS_Cosmetic_Pickaxe_Ready
 bOverrideConcurrency=true
 bPrimeOnLoad=true
 maxDistance=4950
 pitchMultiplier=1.2
 soundClassObject=&Fort_WeaponMelee
 volumeMultiplier=0.15
 concurrencyOverrides {resolutionRule=PreventNew}
}
